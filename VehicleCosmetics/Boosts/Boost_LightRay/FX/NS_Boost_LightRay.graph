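# Niagara system for the LightRay vehicle boost cosmetic, emitting light trails, particles and flares from skeletal mesh sockets, shifted by a car offset.
# A multi-renderer ribbon emitter uses age-driven sine offsets for weaving cross and lightning strands.
# Three sprite emitters add curl-noise particles, rotating flares and short-lived glows, using spawn groups to vary layers within each.
# Painted-variant modules set per-layer colors, with extra tweaks for painted and frontend-preview states.
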NiagaraSystem /VehicleCosmetics/Boosts/Boost_LightRay/FX/NS_Boost_LightRay
dependencies {
 usedDataInterfaces [
  &NiagaraDataInterfaceCurve
  &NiagaraDataInterfaceColorCurve
  &NiagaraDataInterfaceArrayFloat
  &NiagaraDataInterfaceArrayColor
  &NiagaraDataInterfaceSkeletalMesh
 ]
 usedDynamicInputs [&MakeFloatFromVector2d]
 usedModules [
  &self:SystemSpawnScript.SystemScriptSource.SystemScriptGraph.NiagaraNodeAssignment_0.SetVariables_#1
  &SystemState
 ]
}
emitters {
 FlareGroup001 {
  emitterUpdateScript [
   &EmitterState(Inactive Response=::NewEnumerator1, Life Cycle Mode=::NewEnumerator1, Loop Duration=1, UseLoopDelay=false)
   &VC_FilteredSocketCount()
   &SpawnRate(SpawnRate=10, Use Spawn Probability=false)
   &SpawnRate(SpawnGroup=1, SpawnRate=10, Use Spawn Probability=false) {module=SpawnRate001}
   &SpawnRate(SpawnGroup=2, SpawnRate=10, Use Spawn Probability=false) {module=SpawnRate002}
  ]
  particleSpawnScript [
   &VC_PaintedVariantsOverride(DefaultColor=<rgba 0.3944 0.3 2 1>) {enabled=false}
   &VC_PaintedVariantsOverride(
    Black=<rgba 0.22 0.22 0.22 1>
    BurntSienna=<rgba 0.45 0.204 0.09 1>
    Cobalt=<rgba 0.25 0.3333 1.25 1>
    Crimson=<rgba 1.3 0.52 0.546 1>
    DefaultColor=<rgba 0 0.726 1.5 1>
    ForestGreen=<rgba 0.6 1 0.6202 1>
    Grey=<rgba 0.6 0.6 0.6 1>
    Lime=<rgba 0.8 1 0.4 1>
    Orange=<rgba 1.7 0.593 0 1>
    Pink=<rgba 1.6 0.64 1.2509 1>
    Purple=<rgba 0.66 0.44 1.1 1>
    R1 Black:LinearColor=<rgba 0.05 0.05 0.05 1>
    R1 BurntSienna:LinearColor=<rgba 0.3 0.132 0.06 1>
    R1 Cobalt:LinearColor=<rgba 0.52 0.715 1.3 1>
    R1 Crimson:LinearColor=<rgba 1 0.28 0.328 1>
    R1 DefaultColor=<rgba 1.1 0.8653 0.22 1>
    R1 ForestGreen:LinearColor=<rgba 0 1.2 0.0606 1>
    R1 Gold:LinearColor=<rgba 5 5 5 1>
    R1 Grey:LinearColor=<rgba 0.5 0.5 0.5 1>
    R1 Lime:LinearColor=<rgba 0.625 1 0.25 1>
    R1 Orange:LinearColor=<rgba 1 0.4465 0.15 1>
    R1 Pink:LinearColor=<rgba 1 0.4 0.7818 1>
    R1 Purple:LinearColor=<rgba 0.35 0 1.4 1>
    R1 Saffron:LinearColor=<rgba 1 0.9 0 1>
    R1 SkyBlue:LinearColor=<rgba 0.3 0.825 1 1>
    R1 TitaniumWhite:LinearColor=<rgba 2 2 2 1>
    R2 Black:LinearColor=<rgba 0.22 0.22 0.22 1>
    R2 BurntSienna:LinearColor=<rgba 0.65 0.3315 0.195 1>
    R2 Cobalt:LinearColor=<rgba 0.3 0.4167 1 1>
    R2 Crimson:LinearColor=<rgba 1.1 0.594 0.6109 1>
    R2 ForestGreen:LinearColor=<rgba 0.6 1 0.6202 1>
    R2 Grey:LinearColor=<rgba 0.65 0.65 0.65 1>
    R2 Lime:LinearColor=<rgba 0.8 1 0.4 1>
    R2 Orange:LinearColor=<rgba 1.7 0.593 0 1>
    R2 Pink:LinearColor=<rgba 1.7 0.68 1.3291 1>
    R2 Purple:LinearColor=<rgba 0.7367 0.455 1.3 1>
    R2 Saffron:LinearColor=<rgba 2 1.875 0.5 1>
    R2 SkyBlue:LinearColor=<rgba 0.6 1.05 1.2 1>
    R2 TitaniumWhite:LinearColor=<rgba 1 1 1 1>
    Saffron=<rgba 1.25 1.25 0.5625 1>
    SkyBlue=<rgba 0.5 1 1 1>
    TitaniumWhite=<rgba 0.85 0.85 0.85 1>
    Use R1 Painted Overrides=true
    Use R2 Painted Overrides=true
   ) {module=VC_PaintedVariantsOverride001}
   /SetVariables(
    Particles.VisibilityTag=&SpawnGroup()
    Particles.WhiteSize:Vector2f=&Vector2DFromFloat(Value=50)
   ) {isSetParameters=true}
   &InitializeParticle(
    Color Mode=::NewEnumerator0
    Lifetime=1.2
    Sprite Rotation Mode=::NewEnumerator2
    Sprite Size Mode=::NewEnumerator3
    Uniform Sprite Size=85
    UsePositionOffset=false
    Write Material Random=true
   )
   &PartitionParticles(Number Of Partitions=$Emitter.PortCount)
   &SkeletalMeshLocation(
    Bone / Socket Index=$Particles.PartitionParticles.Partition
    Bone / Socket Sampling Mode=::NewEnumerator5
    MaskSpawnGroup=false
    Mesh Sampling Type=::NewEnumerator6
    Sampled Position Offset=&AddVectorToPosition(Position=<-70 0 0>, Vector=Linked() {linkedVariable {name=User.DTCarOffset, type=NiagaraPosition}})
    Skeletal Mesh=$System.SkeletalMesh
   )
   &AddVelocity(Velocity=<-1500 0 0>) {enabled=false}
   /SetVariables(Particles.Corecolor:LinearColor=&MultiplyLinearColor(
     Linear Color=$Particles.Initial.R2 Color
     Scale Factor=&MakeVector4FromVectorAndFloat(
      "Vector (RGB)"=&VectorFromFloat(Value=&MakeCustomFloatFromBool(Bool=$User.bIsPainted, True Float=1.05, False Float=1))
      "Float (Alpha)"=1
     )
    )) {isSetParameters=true}
  ]
  particleUpdateScript [
   &ParticleState()
   &SpriteRotationRate(Rotation Rate=-10)
   &ScaleSpriteSize(
    Uniform Curve Scale=&MergeHelper_FloatSwitch(0=1, 1=0.8, 2=1, 3=?, Index=&SpawnGroup(), Num=::NewEnumerator1)
    Uniform Curve Sprite Scale=DataInterface() {propertyValues {Curve {keys [<at 0 0 RCIM_Cubic>, {tangentMode=RCTM_User, time=1, value=2}]}}}
   )
   &SolveForcesAndVelocity(Clamp Velocity=false, Limit Acceleration=false)
   &DynamicMaterialParameters(
    Index 0 Param 1=/VehicleCosmetics/FX/Niagara/Modules/MergeEmitters/VC_FloatBySpawnGroup(
     SpawnGroup or ExecIndex:NiagaraInt32=0
     CustomInt:NiagaraInt32=0
     MaximumSpawnGroups:NiagaraInt32=3
     Float SG 0:NiagaraFloat=&FloatFromCurve(FloatCurve=DataInterface() {propertyValues {Curve {keys [<at 0 0>,<at 1 0.25>]}}})
     Float SG 1:NiagaraFloat=&FloatFromCurve(FloatCurve=DataInterface() {propertyValues {Curve {keys [<at 0 0>,<at 1 0.25>]}}})
     Float SG 2:NiagaraFloat=&FloatFromCurve(FloatCurve=DataInterface() {propertyValues {Curve {keys [<at 0 0>,<at 1 0.25>]}}})
     Float SG 3:NiagaraFloat=&FloatFromCurve(FloatCurve=DataInterface() {propertyValues {Curve {keys [<at 0 8>,<at 1 8>]}}})
     Float SG 4:NiagaraFloat=0
     Float SG 5:NiagaraFloat=0
    )
    Index 0 Param 2=&VC_FloatBySpawnGroup(
     SpawnGroup or ExecIndex:NiagaraInt32=0
     CustomInt:NiagaraInt32=0
     MaximumSpawnGroups:NiagaraInt32=5
     Float SG 0:NiagaraFloat=&FloatFromCurve(FloatCurve=DataInterface() {propertyValues {Curve {keys [<at 0.7 8>,<at 1 16>]}}})
     Float SG 1:NiagaraFloat=6
     Float SG 2:NiagaraFloat=&FloatFromCurve(FloatCurve=DataInterface() {propertyValues {Curve {keys [<at 0.7 8>,<at 1 16>]}}})
     Float SG 3:NiagaraFloat=&FloatFromCurve(FloatCurve=DataInterface() {propertyValues {Curve {keys [<at 0 6>,<at 1 6>]}}})
     Float SG 4:NiagaraFloat=&FloatFromCurve(FloatCurve=DataInterface() {propertyValues {Curve {keys [<at 0 8>,<at 1 8>]}}})
     Float SG 5:NiagaraFloat=0
    )
    Index 0 Param 3=&VC_FloatBySpawnGroup(
     SpawnGroup or ExecIndex:NiagaraInt32=0
     CustomInt:NiagaraInt32=0
     MaximumSpawnGroups:NiagaraInt32=5
     Float SG 0:NiagaraFloat=0
     Float SG 1:NiagaraFloat=0
     Float SG 2:NiagaraFloat=1
     Float SG 3:NiagaraFloat=0
     Float SG 4:NiagaraFloat=0
     Float SG 5:NiagaraFloat=0
    )
    Index 0 Param 4=&VC_FloatBySpawnGroup(
     SpawnGroup or ExecIndex:NiagaraInt32=0
     CustomInt:NiagaraInt32=0
     MaximumSpawnGroups:NiagaraInt32=5
     Float SG 0:NiagaraFloat=&FloatFromCurve(FloatCurve=DataInterface() {propertyValues {Curve {keys [<at 0 0.5>,<at 1 0.3>]}}})
     Float SG 1:NiagaraFloat=&FloatFromCurve(FloatCurve=DataInterface() {propertyValues {Curve {keys [<at 0 1.2>,<at 1 0.1>]}}})
     Float SG 2:NiagaraFloat=&FloatFromCurve(FloatCurve=DataInterface() {propertyValues {Curve {keys [<at 0 0.3>,<at 1 0.3>]}}})
     Float SG 3:NiagaraFloat=&FloatFromCurve(FloatCurve=DataInterface() {propertyValues {Curve {keys [<at 0 0.03>,<at 1 0.3>]}}})
     Float SG 4:NiagaraFloat=&FloatFromCurve(FloatCurve=DataInterface() {propertyValues {Curve {keys [<at 0 0>,<at 1 0>]}}})
     Float SG 5:NiagaraFloat=&FloatFromCurve(FloatCurve=DataInterface() {propertyValues {Curve {keys [<at 0 1>,<at 1 0>]}}})
    )
    Module.Index0Float0Write=true
    Module.Index0Float1Write=true
    Module.Index0Float2Write=false
    Module.Index0Float3Write=true
   )
   &ScaleColor(
    Scale Alpha=&MakeCustomFloatFromBool(
     Bool=$User.bIsPainted
     True Float=&FloatFromCurve(FloatCurve=DataInterface() {propertyValues {Curve {keys [<at 0 1>,<at 1 0>]}}})
     False Float=1
    )
    Scale RGB=&VectorFromFloat(Value=&MakeCustomFloatFromBool(
      Bool=$User.bIsPainted
      True Float=&FloatFromCurve(FloatCurve=DataInterface() {propertyValues {Curve {defaultValue=0, keys [<at 0 0 RCIM_Cubic>,<at 1 1 RCIM_Cubic>]}}})
      False Float=1
     ))
    ScaleA=true
    ScaleRGB=false
   )
  ]
  rendererClasses [&NiagaraSpriteRendererProperties]
  renderers [
   {rendererClass=&NiagaraSpriteRendererProperties}
   {rendererClass=&NiagaraSpriteRendererProperties, rendererIndex=1}
   {rendererClass=&NiagaraSpriteRendererProperties, rendererIndex=2}
  ]
 }
 NE_VC_Boost_Ribbon_Parent {
  emitterUpdateScript [
   &EmitterState()
   &VC_FilteredSocketCount()
   &VC_RateQualityHandler()
   &VC_SpawnRibbon_v1(Number of Ribbons=$Emitter.PortCount, SpawnRate=60)
  ]
  particleSpawnScript [
   &VC_PaintedVariantsOverride(
    Black=<rgba 0.05 0.05 0.05 1>
    BurntSienna=<rgba 0.15 0.0375 0 1>
    Cobalt=<rgba 0.18 0.288 1.8 1>
    Crimson=<rgba 1.3 0.13 0.1617 1>
    DefaultColor=<rgba 1 0.7783 0.05 0.5913>
    ForestGreen=<rgba 0.2 1 0.2404 1>
    Grey=<rgba 0.28 0.28 0.28 1>
    Lime=<rgba 0.65 1 0.3 1>
    Orange=<rgba 1.2 0.5 0 1>
    Pink=<rgba 1 0.25 0.7273 1>
    Purple=<rgba 0.4 0.2 0.8 1>
    R1 Black:LinearColor=&Multiply_LinearColorByFloat(Linear Color=<rgba 0 0 0 1>, Float=3)
    R1 BurntSienna:LinearColor=<rgba 0.013 0.004 0.0016 3>
    R1 Cobalt:LinearColor=<rgba 0 0.0185 1 0.5>
    R1 Crimson:LinearColor=<rgba 0.464 0.1392 0.148 0.5>
    R1 DefaultColor=&Multiply_LinearColorByFloat(
     Linear Color=&MakeCustomLinearColorFromBool(Bool=$User.bIsFrontendPreview, True Color=<rgba 0 0.0123 0.15 1>, False Color=<rgba 0 0.1083 0.5 0.5>)
     Float=2
    )
    R1 ForestGreen:LinearColor=<rgba 0 0.5 0.0253 0.5>
    R1 Grey:LinearColor=<rgba 0.1 0.1 0.1 2>
    R1 Lime:LinearColor=<rgba 0.2943 0.5885 0 0.2>
    R1 Orange:LinearColor=<rgba 0.6 0.15 0 0.8>
    R1 Pink:LinearColor=<rgba 1 0.3 0.7455 0.2>
    R1 Purple:LinearColor=<rgba 0.0787 0.0213 0.71 1.3>
    R1 Saffron:LinearColor=<rgba 0.7 0.5833 0 0.5>
    R1 SkyBlue:LinearColor=<rgba 0 0.357 0.5365 0.25>
    R1 TitaniumWhite:LinearColor=<rgba 0.5 0.5 0.5 0.35>
    R2 Black:LinearColor=&Multiply_LinearColorByFloat(Linear Color=<rgba 0 0 0 1>, Float=3)
    R2 BurntSienna:LinearColor=<rgba 0.03 0.0101 0.006 1>
    R2 Cobalt:LinearColor=<rgba 0 0.025 1.5 0.5>
    R2 Crimson:LinearColor=<rgba 0.6 0.03 0.049 0.7>
    R2 DefaultColor:LinearColor=<rgba 0 2.8619 4 0.2261>
    R2 ForestGreen:LinearColor=<rgba 0 0.6 0.0303 1>
    R2 Grey:LinearColor=<rgba 0.14 0.14 0.14 0.8>
    R2 Lime:LinearColor=<rgba 0.315 0.6 0.03 1>
    R2 Orange:LinearColor=<rgba 0.9 0.165 0 0.8>
    R2 Pink:LinearColor=<rgba 0.8 0.096 0.544 0.8>
    R2 Purple:LinearColor=<rgba 0.0875 0.05 0.5 0.4>
    R2 Saffron:LinearColor=<rgba 1 0.8333 0 0.75>
    R2 SkyBlue:LinearColor=<rgba 0 0.5 1 0.7>
    R2 TitaniumWhite:LinearColor=<rgba 2 2 2 0.25>
    Saffron=<rgba 1.2 1.004 0.36 1>
    SkyBlue=<rgba 0.25 0.749 1 1>
    TitaniumWhite=<rgba 0.3 0.3 0.3 1>
    Use R1 Painted Overrides=true
    Use R2 Painted Overrides=true
   )
   &VC_InitializeRibbon_v1(
    Color=$Particles.Color
    Custom Offset=true
    Enable Skeletal Position=true
    Local Vector Offset=&AddVectorToPosition(Position=<-70 0 0>, Vector=Linked() {linkedVariable {name=User.DTCarOffset, type=NiagaraPosition}})
    Offset Amount=1
    Ribbon Width=60
    Write Material Random=false
   )
   &AddVelocity(Velocity=<-2000 0 0>) {module=AddVelocity001}
  ]
  particleUpdateScript [
   &ParticleState()
   &InheritSourceMovement(ApplyForce=false, ApplyPosition=true, ApplyPositionFalloff=false)
   &SolveForcesAndVelocity(Clamp Velocity=false, Limit Acceleration=false)
   &VC_UpdateRibbon_v1(Enable Skeletal Position=true)
   &ScaleColor(
    Scale Alpha=&FloatFromCurve(
     FloatCurve=DataInterface() {propertyValues {Curve {
        defaultValue=1
        keys [
         {arriveTangent=3.8118, interpMode=RCIM_Cubic, leaveTangent=3.8118, tangentMode=RCTM_User, time=0, value=0}
         {arriveTangent=1.5427, interpMode=RCIM_Cubic, leaveTangent=-0.033, tangentMode=RCTM_Break, time=0.1789, value=0.9847}
         {arriveTangent=-5.5211, interpMode=RCIM_Cubic, leaveTangent=-5.5211, tangentMode=RCTM_User, time=0.8734, value=0}
        ]
       }}}
     Scale Curve=20
    )
    ScaleA=true
    ScaleRGB=false
   )
   /SetVariables(
    Particles.BeamColor:LinearColor=&MultiplyLinearColor(
     Linear Color=$Particles.R1 Color
     Scale Factor=&MakeVector4FromVectorAndFloat(
      "Vector (RGB)"=<1 1 1>
      "Float (Alpha)"=&FloatFromCurve(FloatCurve=DataInterface() {propertyValues {Curve {
          defaultValue=1
          keys [
           {arriveTangent=1.4927, interpMode=RCIM_Cubic, time=-0.0449, value=0.734}
           {arriveTangent=1.5427, interpMode=RCIM_Cubic, leaveTangent=-1.0354, tangentMode=RCTM_Break, time=0.1789, value=0.9847}
           {arriveTangent=-3.3735, interpMode=RCIM_Cubic, leaveTangent=-3.3735, tangentMode=RCTM_User, time=1.0198, value=-0.2103}
          ]
         }}})
     )
    )
    Particles.BlueWidth:NiagaraFloat=&FloatFromCurve(
     FloatCurve=DataInterface() {propertyValues {Curve {
        defaultValue=1
        keys [
         <at -0.0085 0.7008 RCIM_Cubic>
         {arriveTangent=0.0448, interpMode=RCIM_Cubic, leaveTangent=0.0448, tangentMode=RCTM_User, time=0.2772, value=0.9847}
         {arriveTangent=-0.932, interpMode=RCIM_Cubic, leaveTangent=-0.932, tangentMode=RCTM_User, time=1.0722, value=0.538}
        ]
       }}}
     Scale Curve=150
    )
    Particles.CrossColor:LinearColor=&MakeCustomLinearColorFromBool(
     Bool=$User.bIsPainted
     True Color=&MultiplyLinearColor(
      Linear Color=$Particles.R2 Color
      Scale Factor=&MakeVector4FromVectorAndFloat(
       "Vector (RGB)"=<1 1 1>
       "Float (Alpha)"=&FloatFromCurve(
        FloatCurve=DataInterface() {propertyValues {Curve {
           defaultValue=1
           keys [
            {arriveTangent=1.4927, interpMode=RCIM_Cubic, time=0, value=0}
            {arriveTangent=0.1055, interpMode=RCIM_Cubic, leaveTangent=-1.0354, tangentMode=RCTM_Break, time=0.2653, value=1}
            {arriveTangent=-0.2248, interpMode=RCIM_Cubic, leaveTangent=-0.2248, tangentMode=RCTM_User, time=0.9436, value=0}
           ]
          }}}
        Scale Curve=3
       )
      )
     )
     False Color=&ColorFromCurve(ColorCurve=DataInterface() {propertyValues {
        AlphaCurve {keys [<at 0 0 RCIM_Cubic>,<at 0.4186 1>,<at 1 2 RCIM_Cubic>]}
        BlueCurve {keys [<at 0 1 RCIM_Cubic>,<at 0.7607 1 RCIM_Cubic>]}
        GreenCurve {keys [<at 0 0.6654 RCIM_Cubic>,<at 0.7607 0 RCIM_Cubic>]}
        RedCurve {keys [<at 0 0 RCIM_Cubic>,<at 0.7607 0.029 RCIM_Cubic>]}
       }})
    )
    Particles.CrossColor2:LinearColor=&MakeCustomLinearColorFromBool(
     Bool=$User.bIsFrontendPreview
     True Color=&MultiplyLinearColor(
      Linear Color=$Particles.R1 Color
      Scale Factor=&MakeVector4FromVectorAndFloat(
       "Vector (RGB)"=<1 1 1>
       "Float (Alpha)"=&FloatFromCurve(FloatCurve=DataInterface() {propertyValues {Curve {
           defaultValue=1
           keys [
            {arriveTangent=1.4927, interpMode=RCIM_Cubic, time=-0.0449, value=0.734}
            {arriveTangent=1.5427, interpMode=RCIM_Cubic, leaveTangent=-1.0354, tangentMode=RCTM_Break, time=0.1789, value=0.9847}
            {arriveTangent=-3.3735, interpMode=RCIM_Cubic, leaveTangent=-3.3735, tangentMode=RCTM_User, time=1.0198, value=-0.2103}
           ]
          }}})
      )
     )
     False Color=&MultiplyLinearColor(
      Linear Color=$Particles.R1 Color
      Scale Factor=&MakeVector4FromVectorAndFloat(
       "Vector (RGB)"=<1 5 2>
       "Float (Alpha)"=&FloatFromCurve(
        FloatCurve=DataInterface() {propertyValues {Curve {
           defaultValue=1
           keys [
            {arriveTangent=1.4927, interpMode=RCIM_Cubic, time=-0.0449, value=0.734}
            {arriveTangent=1.5427, interpMode=RCIM_Cubic, leaveTangent=-1.0354, tangentMode=RCTM_Break, time=0.1789, value=0.9847}
            {arriveTangent=-3.3735, interpMode=RCIM_Cubic, leaveTangent=-3.3735, tangentMode=RCTM_User, time=1.0198, value=-0.2103}
           ]
          }}}
        Scale Curve=0.5
       )
      )
     )
    )
    Particles.CrossInvertWidth:NiagaraFloat=&FloatFromCurve(
     FloatCurve=DataInterface() {propertyValues {Curve {
        defaultValue=1
        keys [
         <at -0.0085 0.7008 RCIM_Cubic>
         {arriveTangent=0.0448, interpMode=RCIM_Cubic, leaveTangent=0.0448, tangentMode=RCTM_User, time=0.2772, value=0.9847}
         {arriveTangent=-0.0405, interpMode=RCIM_Cubic, leaveTangent=-0.0405, tangentMode=RCTM_User, time=0.8385, value=0.0052}
        ]
       }}}
     Scale Curve=10
    )
    Particles.CrossWidth:NiagaraFloat=&FloatFromCurve(
     FloatCurve=DataInterface() {propertyValues {Curve {
        defaultValue=1
        keys [
         <at -0.0085 0.7008 RCIM_Cubic>
         {arriveTangent=0.0448, interpMode=RCIM_Cubic, leaveTangent=0.0448, tangentMode=RCTM_User, time=0.2772, value=0.9847}
         {arriveTangent=-0.3044, interpMode=RCIM_Cubic, leaveTangent=-0.3044, tangentMode=RCTM_User, time=0.8, value=0}
        ]
       }}}
     Scale Curve=50
    )
    Particles.LightningColor:LinearColor=&MultiplyLinearColor(
     Linear Color=$Particles.Color
     Scale Factor=&MakeVector4FromVectorAndFloat(
      "Vector (RGB)"=<1 1 1>
      "Float (Alpha)"=&FloatFromCurve(FloatCurve=DataInterface() {propertyValues {Curve {
          defaultValue=1
          keys [
           {arriveTangent=0.0901, interpMode=RCIM_Cubic, leaveTangent=0.0901, tangentMode=RCTM_User, time=-0.0449, value=-0.0307}
           {arriveTangent=-0.4524, interpMode=RCIM_Cubic, leaveTangent=0.0385, tangentMode=RCTM_Break, time=0.5314, value=0.9847}
           {arriveTangent=-0.1302, interpMode=RCIM_Cubic, leaveTangent=-0.1302, tangentMode=RCTM_User, time=1.4722, value=0}
          ]
         }}})
     )
    )
    Particles.ThunderWidth:NiagaraFloat=&FloatFromCurve(
     FloatCurve=DataInterface() {propertyValues {Curve {
        defaultValue=1
        keys [
         {arriveTangent=1.4927, interpMode=RCIM_Cubic, time=-0.0449, value=0.734}
         {arriveTangent=1.5427, interpMode=RCIM_Cubic, leaveTangent=-0.1679, tangentMode=RCTM_Break, time=0.1789, value=0.9847}
         {arriveTangent=-2.714, interpMode=RCIM_Cubic, leaveTangent=-2.714, tangentMode=RCTM_User, time=1.0781, value=-0.1679}
        ]
       }}}
     Scale Curve=50
    )
   ) {isSetParameters=true}
   /SetVariables(
    Particles.CrossInvertPosition:NiagaraPosition=&AddVectorToPosition(
     Position=$Particles.Position
     Vector=&TransformVector(
      Vector=&MakeVector(
       X=&SelectFloatFromArray(Direct Array Index=&SpawnGroup(), Float Selection Array=DataInterface() {propertyValues {FloatData [-30,0]}})
       Y=0
       Z=&Sine(
        Normalized Angle=&Add_Float(
         A=&Multiply_Float(A=$Particles.NormalizedAge, B=-1.3)
         B=&Multiply_Float(A=$Emitter.LoopedAge, B=1.2)
        )
        Period=-1
        Scale=&FloatFromCurve(
         FloatCurve=DataInterface() {propertyValues {Curve {
            defaultValue=1
            keys [
             {arriveTangent=2.952, interpMode=RCIM_Cubic, leaveTangent=2.952, tangentMode=RCTM_User, time=-0.0185, value=0}
             {arriveTangent=0.1401, tangentMode=RCTM_User, time=0.387, value=0.6595}
             {arriveTangent=-1.5773, interpMode=RCIM_Cubic, leaveTangent=-3.8709, tangentMode=RCTM_Break, time=0.5819, value=0.6595}
             {arriveTangent=-0.2749, interpMode=RCIM_Cubic, leaveTangent=-0.2749, tangentMode=RCTM_User, time=1.0012, value=0}
            ]
           }}}
         Scale Curve=20
        )
        Bias=0
       )
      )
      Destination Space=::Simulation
      Apply Scale=true
     )
    )
    Particles.PositionCross:NiagaraPosition=&AddVectorToPosition(
     Position=$Particles.Position
     Vector=&TransformVector(
      Vector=&MakeVector(
       X=&SelectFloatFromArray(Direct Array Index=&SpawnGroup(), Float Selection Array=DataInterface() {propertyValues {FloatData [-30,0]}})
       Y=0
       Z=&Sine(
        Normalized Angle=&Add_Float(
         A=&Multiply_Float(A=$Particles.NormalizedAge, B=-1.3)
         B=&Multiply_Float(A=$Emitter.LoopedAge, B=1.5)
        )
        Period=1
        Scale=&FloatFromCurve(
         FloatCurve=DataInterface() {propertyValues {Curve {
            defaultValue=1
            keys [
             {arriveTangent=-0.1097, interpMode=RCIM_Cubic, leaveTangent=-0.1097, tangentMode=RCTM_User, time=0.175, value=0}
             {arriveTangent=7.2017, leaveTangent=7.2017, time=0.3137, value=0.7182}
             {arriveTangent=-2.5934, interpMode=RCIM_Cubic, leaveTangent=-2.5934, tangentMode=RCTM_User, time=0.4951, value=0.7335}
             {arriveTangent=-0.0405, interpMode=RCIM_Cubic, leaveTangent=-0.0405, tangentMode=RCTM_User, time=0.7, value=0}
            ]
           }}}
         Scale Curve=15
        )
        Bias=0
       )
      )
      Destination Space=::Simulation
      Apply Scale=true
     )
    )
    Particles.PositionLightning:NiagaraPosition=&AddVectorToPosition(
     Position=$Particles.Position
     Vector=&TransformVector(
      Vector=&MakeVector(
       X=0
       Y=0
       Z=&Sine(
        Normalized Angle=&Add_Float(
         A=&Multiply_Float(A=$Particles.NormalizedAge, B=-1.3)
         B=&Multiply_Float(A=$Emitter.LoopedAge, B=1.3)
        )
        Period=1
        Scale=&FloatFromCurve(
         FloatCurve=DataInterface() {propertyValues {Curve {
            defaultValue=1
            keys [
             {arriveTangent=9.0725, interpMode=RCIM_Cubic, leaveTangent=9.0725, tangentMode=RCTM_User, time=-0.0048, value=0}
             {arriveTangent=0.0244, interpMode=RCIM_Cubic, leaveTangent=0.0244, tangentMode=RCTM_User, time=0.1408, value=0.9847}
             {arriveTangent=-0.0583, interpMode=RCIM_Cubic, leaveTangent=-0.0583, tangentMode=RCTM_User, time=0.3728, value=1}
             {arriveTangent=-0.2024, interpMode=RCIM_Cubic, leaveTangent=-0.2024, tangentMode=RCTM_User, time=1.1034, value=0}
            ]
           }}}
         Scale Curve=13
        )
        Bias=0
       )
      )
      Destination Space=::Simulation
      Apply Scale=true
     )
    )
   ) {isSetParameters=true}
  ]
  rendererClasses [&NiagaraRibbonRendererProperties]
  renderers [
   {rendererClass=&NiagaraRibbonRendererProperties}
   {rendererClass=&NiagaraRibbonRendererProperties, rendererIndex=1}
   {rendererClass=&NiagaraRibbonRendererProperties, rendererIndex=2}
   {rendererClass=&NiagaraRibbonRendererProperties, rendererIndex=3}
  ]
 }
 NE_VC_Boost_Sprite_Parent {
  emitterUpdateScript [
   &EmitterState()
   &VC_FilteredSocketCount()
   &SpawnRate(SpawnRate=60, Use Spawn Probability=false)
   &SpawnRate(SpawnGroup=1, SpawnRate=30, Use Spawn Probability=false) {module=SpawnRate001}
  ]
  particleSpawnScript [
   &VC_PaintedVariantsOverride(
    Black=<rgba 0.1 0.1 0.1 0.5>
    BurntSienna=<rgba 0.4 0.1477 0.08 0.2174>
    Cobalt=<rgba 0.2 0.2148 1 0.4174>
    Crimson=<rgba 1 0.6 0.6108 0.1391>
    DefaultColor=<rgba 1 0.8339 0.1261 0.25>
    ForestGreen=<rgba 0.2217 0.5 0.2358 0.7565>
    Grey=<rgba 0.7043 0.7043 0.7043 0.3217>
    Lime=<rgba 0.4925 0.6783 0.3067 0.8435>
    Orange=<rgba 1 0.3488 0 0.6522>
    Pink=<rgba 0.8783 0.3666 0.6922 0.2522>
    Purple=<rgba 0.25 0.1 1 0.6>
    R1 Black:LinearColor=<rgba 1 1 1 0.2>
    R1 BurntSienna:LinearColor=<rgba 1 0.431 0.2783 0.4435>
    R1 Cobalt:LinearColor=<rgba 0.45 0.4694 1.5 1>
    R1 Crimson:LinearColor=<rgba 2 0.3478 0.3926 1>
    R1 DefaultColor=<rgba 2 1.6951 0.3958 0.8>
    R1 ForestGreen:LinearColor=<rgba 0 1 0.0505 1>
    R1 Grey:LinearColor=<rgba 0.5391 0.5391 0.5391 1>
    R1 Orange:LinearColor=<rgba 2 1.2928 0 1>
    R1 Pink:LinearColor=<rgba 0.887 0.3085 0.6766 1>
    R1 Purple:LinearColor=<rgba 0.7 0.3 1.5 2>
    R1 Saffron:LinearColor=<rgba 1 0.8255 0.2522 1>
    R1 SkyBlue:LinearColor=<rgba 0.0976 0.4397 0.4179 1>
    R1 TitaniumWhite:LinearColor=<rgba 1 1 1 1>
    R2 Black:LinearColor=<rgba 0.1 0.1 0.1 0.5>
    R2 BurntSienna:LinearColor=<rgba 0.4 0.1477 0.08 0.2174>
    R2 Cobalt:LinearColor=<rgba 0.2 0.2148 1 0.4174>
    R2 Crimson:LinearColor=<rgba 1 0.6 0.6108 0.1391>
    R2 DefaultColor:LinearColor=<rgba 1 0.8339 0.1261 1>
    R2 ForestGreen:LinearColor=<rgba 0.2217 0.5 0.2358 0.7565>
    R2 Grey:LinearColor=<rgba 0.7043 0.7043 0.7043 0.3217>
    R2 Lime:LinearColor=<rgba 0.4925 0.6783 0.3067 0.8435>
    R2 Orange:LinearColor=<rgba 1 0.3488 0 0.6522>
    R2 Pink:LinearColor=<rgba 0.8783 0.3666 0.6922 0.2522>
    R2 Purple:LinearColor=<rgba 0.2042 0.105 0.7 0.5217>
    R2 Saffron:LinearColor=<rgba 1 0.8417 0.3217 0.287>
    R2 SkyBlue:LinearColor=<rgba 0.4261 1 0.9635 0.3652>
    R2 TitaniumWhite:LinearColor=<rgba 0.75 0.75 0.75 0.2957>
    Saffron=<rgba 1 0.8417 0.3217 0.287>
    SkyBlue=<rgba 0.4261 1 0.9635 0.3652>
    TitaniumWhite=<rgba 0.75 0.75 0.75 0.2957>
    Use R1 Painted Overrides=true
    Use R2 Painted Overrides=true
   )
   &InitializeParticle(
    Color=$Particles.Color
    Lifetime=0.6
    Sprite Rotation Mode=::NewEnumerator2
    Sprite Size Mode=::NewEnumerator1
    Uniform Sprite Size Max=30
    Uniform Sprite Size Min=10
    UsePositionOffset=false
    Write Material Random=true
   )
   &PartitionParticles(Number Of Partitions=$Emitter.PortCount)
   &SkeletalMeshLocation(
    Bone / Socket Index=$Particles.PartitionParticles.Partition
    Bone / Socket Sampling Mode=::NewEnumerator5
    MaskSpawnGroup=false
    Mesh Sampling Type=::NewEnumerator6
    Sampled Position Offset=&AddVectorToPosition(Position=<-70 0 0>, Vector=Linked() {linkedVariable {name=User.DTCarOffset, type=NiagaraPosition}})
    Skeletal Mesh=$System.SkeletalMesh
   )
   &AddVelocity(Velocity=<-800 0 0>)
   &ShapeLocation(MaskSpawnGroup=false, Override Seed=false, Sphere Radius=30, UseOwnerScale=true)
   /SetVariables(Particles.BloomSize:Vector2f=<25 25>, Particles.VisibilityTag=&SpawnGroup()) {isSetParameters=true}
  ]
  particleUpdateScript [
   &ParticleState()
   &ScaleSpriteSize(Uniform Curve Sprite Scale=DataInterface() {propertyValues {Curve {
       defaultValue=0
       keys [{interpMode=RCIM_Cubic, tangentMode=RCTM_User, time=0.2856, value=1}, {interpMode=RCIM_Cubic, tangentMode=RCTM_User, time=1, value=0}]
      }}})
   &Drag(
    Drag=&FloatFromCurve(
     FloatCurve=DataInterface() {propertyValues {Curve {
        defaultValue=1
        keys [
         {arriveTangent=0.4861, interpMode=RCIM_Cubic, leaveTangent=0.4861, tangentMode=RCTM_User, time=0, value=0}
         {interpMode=RCIM_Cubic, tangentMode=RCTM_User, time=0.83, value=1}
        ]
       }}}
     Scale Curve=5
    )
    Use Linear Drag=true
    Use Rotational Drag=true
   )
   &CurlNoiseForce(
    Bias Noise=false
    Noise Frequency=450
    Noise Strength=&RandomRangeFloat(Minimum=250, Maximum=350)
    Pan Noise=false
   ) {module=CurlNoiseForce001}
   &InheritSourceMovement(ApplyForce=false, ApplyPosition=true, ApplyPositionFalloff=false)
   &SolveForcesAndVelocity(Clamp Velocity=false, Limit Acceleration=false)
   /SetVariables(Particles.BloomColor:LinearColor=&MultiplyLinearColor(
     Linear Color=$Particles.R1 Color
     Scale Factor=&MakeVector4FromVectorAndFloat(
      "Vector (RGB)"=<1 1 1>
      "Float (Alpha)"=&FloatFromCurve(
       FloatCurve=DataInterface() {propertyValues {Curve {
          defaultValue=1
          keys [{interpMode=RCIM_Cubic, tangentMode=RCTM_User, time=0.6766, value=1}, {interpMode=RCIM_Cubic, tangentMode=RCTM_User, time=0.7719, value=0}]
         }}}
       Scale Curve=1.3
      )
     )
    )) {isSetParameters=true}
  ]
  rendererClasses [&NiagaraSpriteRendererProperties]
  renderers [
   {rendererClass=&NiagaraSpriteRendererProperties}
   {rendererClass=&NiagaraSpriteRendererProperties, rendererIndex=1}
   {rendererClass=&NiagaraSpriteRendererProperties, rendererIndex=2}
  ]
 }
 NE_VC_Boost_Sprite_Parent001 {
  emitterUpdateScript [
   &EmitterState(Inactive Response=::NewEnumerator1, Life Cycle Mode=::NewEnumerator1, Loop Duration=0.5, UseLoopDelay=false)
   &VC_FilteredSocketCount()
   &SpawnRate(SpawnRate=40, Use Spawn Probability=false)
   &SpawnRate(SpawnGroup=1, SpawnRate=40, Use Spawn Probability=false) {module=SpawnRate001}
  ]
  particleSpawnScript [
   &VC_PaintedVariantsOverride(
    Black=<rgba 0.5 0.5 0.5 0.05>
    BurntSienna=<rgba 0.25 0.0813 0.025 0.6>
    Cobalt=<rgba 0.4 0.6667 2 0.1>
    Crimson=<rgba 0.4635 0.198 0.2052 0>
    DefaultColor=&SelectLinearColorFromArray(
     Array Sampling Mode=::NewEnumerator0
     Direct Array Index=?
     Color Selection Array=DataInterface() {propertyValues {ColorData [<rgba 0 0.968 2 0.25>, <rgba 2 1.5567 0.1 0.25>]}}
     Random Seed=0
    )
    ForestGreen=<rgba 0.1957 0.5 0.211 0.2174>
    Grey=<rgba 0.7 0.7 0.7 0.2>
    Lime=<rgba 0.587 1 0.1739 0.4435>
    Orange=<rgba 1 0.52 0.1 0.5>
    Pink=<rgba 1 0.4 0.7818 0.7>
    Purple=<rgba 0.572 0.22 1.1 0.6>
    R1 Black:LinearColor=<rgba 0.03 0.03 0.03 3>
    R1 BurntSienna:LinearColor=<rgba 0.0642 0.0198 0.0079 1>
    R1 Cobalt:LinearColor=<rgba 0 0.0185 1 3>
    R1 Crimson:LinearColor=<rgba 0.4635 0.0967 0.1067 1>
    R1 DefaultColor=<rgba 0.3 1 0.9953 1>
    R1 ForestGreen:LinearColor=<rgba 0 0.3 0.0152 1>
    R1 Grey:LinearColor=<rgba 0.6 0.6 0.6 0.7>
    R1 Orange:LinearColor=<rgba 0.6724 0.3221 0.1345 0.7>
    R1 Pink:LinearColor=<rgba 1 0.3 0.7455 0.7>
    R1 Purple:LinearColor=<rgba 0.2873 0 0.7 1>
    R1 Saffron:LinearColor=<rgba 0.9 0.711 0.09 0.5>
    Saffron=<rgba 1.2 1.0119 0.3938 0.3217>
    SkyBlue=<rgba 0.3 0.7658 1 0.3>
    Use R1 Painted Overrides=true
   )
   /SetVariables(
    Particles.FlareAlignment:Vector3f=<0 1 0>
    Particles.FlareFacing:Vector3f=&RandomRangeVector(Minimum=<0 1 0.1>, Maximum=<0 1 1>)
    Particles.FlareSize:Vector2f=<150 300>
    Particles.VisibilityTag=&SpawnGroup()
   ) {isSetParameters=true}
   &InitializeParticle(
    Color Mode=::NewEnumerator0
    Lifetime=0.2
    Sprite Rotation Angle=90
    Sprite Rotation Mode=::NewEnumerator0
    Sprite Size Mode=::NewEnumerator3
    Uniform Sprite Size=40
    UsePositionOffset=false
    Write Material Random=true
   )
   &PartitionParticles(Number Of Partitions=$Emitter.PortCount, Partition Mode:ENiagara_PartitionMode=::NewEnumerator1)
   &SkeletalMeshLocation(
    Bone / Socket Index=$Particles.PartitionParticles.Partition
    Bone / Socket Sampling Mode=::NewEnumerator5
    MaskSpawnGroup=false
    Mesh Sampling Type=::NewEnumerator6
    Sampled Position Offset=&AddVectorToPosition(
     Position=&MaskVectorBySpawnGroup(Masked Vector=<-70 0 0>, Passthrough Vector=<-190 0 0>)
     Vector=Linked() {linkedVariable {name=User.DTCarOffset, type=NiagaraPosition}}
    )
    Skeletal Mesh=$System.SkeletalMesh
   )
   &AddVelocity(Velocity=<-1500 0 0>) {enabled=false}
   &ShapeLocation(
    MaskSpawnGroup=false
    Override Seed=false
    Sphere Radius=&VC_FloatBySpawnGroup(
     SpawnGroup or ExecIndex:NiagaraInt32=0
     CustomInt:NiagaraInt32=0
     MaximumSpawnGroups:NiagaraInt32=2
     Float SG 0:NiagaraFloat=5
     Float SG 1:NiagaraFloat=20
     Float SG 2:NiagaraFloat=85
     Float SG 3:NiagaraFloat=35
     Float SG 4:NiagaraFloat=45
     Float SG 5:NiagaraFloat=0
    )
    UseOwnerScale=true
   )
   /SetVariables(Particles.FlareColor=&MultiplyLinearColor(
     Linear Color=$Particles.R1 Color
     Scale Factor=&MakeVector4FromVectorAndFloat("Vector (RGB)"=<1 1 1>, "Float (Alpha)"=0.03)
    )) {isSetParameters=true}
  ]
  particleUpdateScript [
   &ParticleState()
   &SolveForcesAndVelocity(Clamp Velocity=false, Limit Acceleration=false)
  ]
  rendererClasses [&NiagaraSpriteRendererProperties]
  renderers [
   {rendererClass=&NiagaraSpriteRendererProperties}
   {rendererClass=&NiagaraSpriteRendererProperties, rendererIndex=1}
  ]
 }
}
summary {
 systemName=NS_Boost_LightRay
 emitters [
  {emitterName=NE_VC_Boost_Ribbon_Parent, rendererClasses [&NiagaraRibbonRendererProperties]}
  {emitterName=NE_VC_Boost_Sprite_Parent, rendererClasses [&NiagaraSpriteRendererProperties]}
  {emitterName=FlareGroup001, rendererClasses [&NiagaraSpriteRendererProperties]}
  {emitterName=NE_VC_Boost_Sprite_Parent001, rendererClasses [&NiagaraSpriteRendererProperties]}
 ]
 userVariables [
  User.AddMultiply {type=NiagaraFloat, defaultValue=0, description=""}
  User.bIsFrontendPreview {type=NiagaraBool}
  User.bIsPainted {type=NiagaraBool}
  User.Brightness {type=NiagaraFloat, defaultValue=0, description=""}
  User.ColorVariantID {type=NiagaraInt32}
  User.DTCarOffset {type=NiagaraPosition, description="", defaultValue {}}
  User.PaintedColor {type=LinearColor, description="", defaultValue {struct=LinearColor, value=<rgba 1 1 1 1>}}
  User.PaintedScalar {type=NiagaraFloat, defaultValue=1, description=""}
  User.PortCount {type=NiagaraInt32, defaultValue=2}
  User.SolidMask {type=NiagaraFloat, defaultValue=0, description=""}
 ]
}
